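# Physics asset for the PrecisionMongoose male medium player cosmetic's face accessory mesh, setting up bodies for a dynamic front hair strand and tail.
# Capsules cover both three-segment hair chains, the front strand's base, and the head, neck and upper spine, with two on the spine.
# Both chains attach to the head, the front one through its base bone, with angle-limited hair joints and tight twist along the segments.
# Spine-neck, neck-head and head-to-front-base constraints are free-rotating.
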PhysicsAsset /BRCosmetics/Characters/Player/Male/Medium/Bodies/M_MED_PrecisionMongoose/Meshes/Parts/M_MED_PrecisionMongoose_FaceAcc_Physics
bodies {
 dyn_hair_f_1 [{shapeType=Capsule, length=2.5687, radius=1.0967, shapeName=dyn_hair_f_1_capsule, center=<1.2715 0 0>, rotation=<rot 90 0 0>}]
 dyn_hair_f_2 [{shapeType=Capsule, length=2.0075, radius=1.0377, shapeName=dyn_hair_f_2_capsule, center=<1.0169 0 0>, rotation=<rot 90 0 0>}]
 dyn_hair_f_3 [{shapeType=Capsule, length=2.0826, radius=0.9625, shapeName=dyn_hair_f_3_capsule, center=<0.6612 0 0>, rotation=<rot 90 0 0>}]
 dyn_hair_f_base [{shapeType=Capsule, length=0.505, radius=0.505, shapeName=dyn_hair_f_base_capsule, rotation=<rot 0 -90 0>}]
 dyn_hair_tail_1 [{shapeType=Capsule, length=0.98, radius=3, shapeName=dyn_hair_tail_1_capsule, center=<0.2179 0 0>, rotation=<rot 90 0 0>}]
 dyn_hair_tail_2 [{shapeType=Capsule, length=3, radius=3, shapeName=dyn_hair_tail_2_capsule, center=<3.0892 -1.3 0>, rotation=<rot 90 0 0>}]
 dyn_hair_tail_3 [{shapeType=Capsule, length=2.6562, radius=2.198, shapeName=dyn_hair_tail_3_capsule, center=<1 -1.5 0>, rotation=<rot 90 0 9.9999>}]
 head [{shapeType=Capsule, length=4.758, radius=6.2535, shapeName=head_capsule, center=<7.5077 1.8349 0>, rotation=<rot 90 0 -13.7343>}]
 neck_02 [{shapeType=Capsule, length=3.8728, radius=7.2028, shapeName=neck_02_capsule, center=<2.7328 -1.2743 0>, rotation=<rot 90 -179.4032 -174.4031>}]
 spine_05 [
  {shapeType=Capsule, length=14.4518, radius=13.5364, shapeName=spine_05_capsule, center=<2.8179 1.559 0>}
  {
   shapeType=Capsule
   length=9.4415
   radius=6.2123
   shapeName=spine_05_capsule
   center=<17.8168 -1.5206 -2.0917>
   rotation=<rot 10.3236 -2.619 9.3126>
  }
 ]
}
constraints [
 {
  bone1Name=neck_02
  bone2Name=spine_05
  swing1LimitDegrees=45
  swing1Motion=Free
  swing2LimitDegrees=45
  swing2Motion=Free
  twistLimitDegrees=45
  twistMotion=Free
 }
 {
  bone1Name=head
  bone2Name=neck_02
  swing1LimitDegrees=45
  swing1Motion=Free
  swing2LimitDegrees=45
  swing2Motion=Free
  twistLimitDegrees=45
  twistMotion=Free
 }
 {bone1Name=dyn_hair_f_2, bone2Name=dyn_hair_f_1, swing1LimitDegrees=25, swing2LimitDegrees=25, twistLimitDegrees=5}
 {bone1Name=dyn_hair_f_3, bone2Name=dyn_hair_f_2, swing1LimitDegrees=25, swing2LimitDegrees=25, twistLimitDegrees=5}
 {bone1Name=dyn_hair_tail_1, bone2Name=head, swing1LimitDegrees=10, swing2LimitDegrees=10, twistLimitDegrees=10}
 {bone1Name=dyn_hair_tail_2, bone2Name=dyn_hair_tail_1, swing1LimitDegrees=35, swing2LimitDegrees=35, twistLimitDegrees=5}
 {bone1Name=dyn_hair_tail_3, bone2Name=dyn_hair_tail_2, swing1LimitDegrees=35, swing2LimitDegrees=35, twistLimitDegrees=5}
 {
  bone1Name=dyn_hair_f_base
  bone2Name=head
  swing1LimitDegrees=45
  swing1Motion=Free
  swing2LimitDegrees=45
  swing2Motion=Free
  twistLimitDegrees=45
  twistMotion=Free
 }
 {bone1Name=dyn_hair_f_1, bone2Name=dyn_hair_f_base, swing1LimitDegrees=25, swing2LimitDegrees=25, twistLimitDegrees=5}
]
